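# Indoor ambience Sound Cue for CliffTown: a looping quad recording of distant metal rattling in wind, reused from Asteria JunkYard, on the indoor ambient sound class.
# The looping wave runs through a Modulator into two nested, indefinitely looping Envelopers, the inner shaping pitch over 30 seconds and the outer swelling and dipping volume over 40.
# The differing periods vary the loop over time, and a boosting Mixer feeds the same chain to all four quality levels.
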SoundCue /Game/Sounds/Helios/POI/CliffTown/Ambience_CliffTown_Metal_Wind_Rattles_Int_Loop_Cue
root=QualityLevel_0
nodes [
 {
  id=Enveloper_12
  children [Modulator_0]
  properties {
   bLoop=true
   bLoopIndefinitely=true
   loopEnd=30
   pitchCurve {editorCurveData {keys [{interpMode=RCIM_Cubic, time=0}, <at 15.9468 0.92 RCIM_Cubic>, {interpMode=RCIM_Cubic, time=30}]}}
   volumeCurve {editorCurveData {keys [<at 0 1>]}}
  }
 }
 {
  id=Enveloper_17
  children [Enveloper_12]
  properties {
   bLoop=true
   bLoopIndefinitely=true
   loopEnd=40
   pitchCurve {editorCurveData {keys [{time=0}]}}
   volumeCurve {editorCurveData {keys [
      {arriveTangent=-0.01, interpMode=RCIM_Cubic, time=0, value=1}
      {arriveTangent=-0.0031, interpMode=RCIM_Cubic, leaveTangent=-0.0031, time=9.8969, value=0.3718}
      {arriveTangent=-0.0105, interpMode=RCIM_Cubic, leaveTangent=-0.0105, tangentMode=RCTM_User, time=18.5053, value=0.9432}
      {arriveTangent=-0.0515, interpMode=RCIM_Cubic, leaveTangent=-0.0515, time=23.4399, value=0.2987}
      {arriveTangent=0.0013, interpMode=RCIM_Cubic, leaveTangent=0.0013, tangentMode=RCTM_User, time=28.1028, value=0.4487}
      {arriveTangent=-0.0006, interpMode=RCIM_Cubic, leaveTangent=-0.0006, tangentMode=RCTM_User, time=32.9324, value=0.3744}
      <at 40 1>
     ]}}
  }
 }
 {id=Mixer_0, children [Enveloper_17], properties {inputVolume [2]}}
 {id=Modulator_0, children [WavePlayer_4], properties {pitchMax=1, pitchMin=1}}
 {id=QualityLevel_0, children [Mixer_0,Mixer_0,Mixer_0,Mixer_0]}
 {
  id=WavePlayer_4
  properties {bLooping=true, soundWaveAssetPtr=/Game/Sounds/Asteria/POI/JunkYard/Ambience_Metal_Distant_Rattling_In_Wind_Loop_03_Quad}
 }
]
properties {
 attenuationSettings=&Fort_Ambient_Athena_Att
 maxDistance=2097152
 soundClassObject=&Fort_Ambient_Indoors_Athena
 volumeMultiplier=0.175
 attenuationOverrides {bAttenuate=false}
}
