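# Sound cue for a looping ambience of metal rattling in wind at the CliffTown POI, reusing a mono JunkYard loop.
# The looping wave passes through a modulator and two nested endlessly looping envelopers, one varying pitch over 30 seconds, the other volume over 40, so the variation drifts.
# A single mixer feeds every quality level.
# Overrides allow one instance and set custom distance falloff with low-pass and listener focus, under the Fort_Earthquakes sound class.
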SoundCue /Game/Sounds/Helios/POI/CliffTown/Ambience_Metal_Wind_Rattles_Loop_Cue
root=QualityLevel_0
nodes [
 {
  id=Enveloper_12
  children [Modulator_0]
  properties {
   bLoop=true
   bLoopIndefinitely=true
   loopEnd=30
   pitchCurve {editorCurveData {keys [{interpMode=RCIM_Cubic, time=0}, <at 15.9468 0.92 RCIM_Cubic>, {interpMode=RCIM_Cubic, time=30}]}}
   volumeCurve {editorCurveData {keys [<at 0 1>]}}
  }
 }
 {
  id=Enveloper_17
  children [Enveloper_12]
  properties {
   bLoop=true
   bLoopIndefinitely=true
   loopEnd=40
   pitchCurve {editorCurveData {keys [{time=0}]}}
   volumeCurve {editorCurveData {keys [
      {arriveTangent=-0.01, interpMode=RCIM_Cubic, time=0, value=1}
      {arriveTangent=-0.0031, interpMode=RCIM_Cubic, leaveTangent=-0.0031, time=9.6527, value=0.8674}
      {arriveTangent=-0.0105, interpMode=RCIM_Cubic, leaveTangent=-0.0105, tangentMode=RCTM_User, time=18.5053, value=0.9432}
      {arriveTangent=0.0001, interpMode=RCIM_Cubic, leaveTangent=0.0001, time=23.1152, value=0.75}
      {arriveTangent=0.0013, interpMode=RCIM_Cubic, leaveTangent=0.0013, tangentMode=RCTM_User, time=27.8586, value=0.9444}
      {arriveTangent=-0.0006, interpMode=RCIM_Cubic, leaveTangent=-0.0006, tangentMode=RCTM_User, time=32.6882, value=0.8701}
      <at 40 1>
     ]}}
  }
 }
 {id=Mixer_0, children [Enveloper_17], properties {inputVolume [1]}}
 {id=Modulator_0, children [WavePlayer_3], properties {pitchMax=1, pitchMin=1}}
 {id=QualityLevel_0, children [Mixer_0,Mixer_0,Mixer_0,Mixer_0]}
 {
  id=WavePlayer_3
  properties {bLooping=true, soundWaveAssetPtr=/Game/Sounds/Asteria/POI/JunkYard/Ambience_Metal_Distant_Rattling_In_Wind_Loop_03_Mono}
 }
]
properties {
 bOverrideAttenuation=true
 bOverrideConcurrency=true
 maxDistance=4400
 soundClassObject=/Game/Sounds/FortSoundClassesAndModes/Fort_Earthquakes
 volumeMultiplier=0.8
 attenuationOverrides {
  bAttenuateWithLPF=true
  bEnableFocusInterpolation=true
  bEnableListenerFocus=true
  bEnableLogFrequencyScaling=true
  bEnableReverbSend=false
  bEnableSubmixSends=true
  coneSphereFalloffDistance=500
  coneSphereRadius=1
  distanceAlgorithm=Custom
  falloffDistance=4000
  hPFFrequencyAtMax=1000
  hPFFrequencyAtMin=300
  lPFRadiusMax=2000
  lPFRadiusMin=500
  nonFocusAzimuth=55
  nonFocusVolumeAttenuation=0.4
  stereoSpread=5
  attenuationShapeExtents=<0 90>
  customAttenuationCurve {editorCurveData {keys [
     <at 0 1>
     {arriveTangent=-0.5069, leaveTangent=-0.5069, time=0.2, value=0.75}
     {arriveTangent=0.0909, interpMode=RCIM_Cubic, leaveTangent=0.0909, time=0.6565, value=0.7}
     {arriveTangent=0.007, interpMode=RCIM_Cubic, leaveTangent=0.007, tangentMode=RCTM_User, time=0.75, value=0.8}
     {arriveTangent=-1.6436, interpMode=RCIM_Cubic, leaveTangent=-1.6436, tangentMode=RCTM_User, time=0.8274, value=0.2624}
     {interpMode=RCIM_Cubic, tangentMode=RCTM_User, time=1.1, value=0}
    ]}}
 }
 concurrencyOverrides {maxCount=1, resolutionRule=PreventNew}
}
